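; Pickaxe Blueprint for the ChiveFlake cosmetic that repositions its impact effect when shown in a frontend context.
; On impact FX playback it calls the parent, checks for frontend display, then gets the swing montage section index from the instigator pawn.
; It moves Impact FX to the pawn mesh's location, then adds a per-section offset from FEImpactOnPawnLocations, rotated by the mesh's world rotation.
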(event EventBeginPlay
  (|CallParentFunction) ; This node is disabled and will not be called.\nDrag off pins to build functionality.
)

(event EventTick
  (params (DeltaSeconds :type Float%20%28single-precision%29)))

(event FortWeapon|EventOnPlayImpactFX
  (params (HitResult :type Hit%20Result%20Structure%20%28by%20ref%29) (ImpactPhysicalSurface :type EPhysicalSurface%20Enum :default "SurfaceType_Default") (SpawnedPSC :type FXSystem%20Component%20Object%20Reference))
  (bind _impact_fx (Variables|Default|GetImpact%20FX))
  (|CallParentFunction HitResult ImpactPhysicalSurface SpawnedPSC)
  (if (Reactive|Frontend|IsCosmeticObjectDisplayedinFrontendContext 0)
    (bind _asfort_pawn (Utilities|Casting|CastToFortPawn (Game|GetInstigator)))
    (bind _section_index (Utilities|SetupSwingMontageSectionIndex _asfort_pawn)
      (:Out
        (Transformation|SetWorldLocation _impact_fx (Transformation|GetWorldLocation (Variables|/Script/Engine.Character|GetMesh _asfort_pawn)) false false)
        (Transformation|AddRelativeLocation _impact_fx (Math|Vector|RotateVector (Utilities|Array|Get(aref) (Variables|Default|GetFEImpactOnPawnLocations) _section_index) (Transformation|GetWorldRotation (Variables|/Script/Engine.Character|GetMesh _asfort_pawn))) false false)))))

(fn ConstructionScript
  (|CallParentFunction))
