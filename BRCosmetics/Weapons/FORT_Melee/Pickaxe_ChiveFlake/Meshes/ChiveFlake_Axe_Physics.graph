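# Physics asset for the ChiveFlake pickaxe cosmetic, setting up physics for the left and right laces and ties on the axe.
# Capsule bodies cover the handle, four base anchor bones, a three-bone dynamic chain per lace and one bone per tie.
# Base bones constrain to the handle with free rotation, while dynamic joints lock twist and limit swing.
# Ties get very tight swing limits, and laces get wider, slightly asymmetric ones.
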PhysicsAsset /BRCosmetics/Weapons/FORT_Melee/Pickaxe_ChiveFlake/Meshes/ChiveFlake_Axe_Physics
bodies {
 base_lace_l [{shapeType=Capsule, length=0.505, radius=0.505, shapeName=base_lace_l_capsule, rotation=<rot 0 -90 0>}]
 base_lace_r [{shapeType=Capsule, length=0.505, radius=0.505, shapeName=base_lace_r_capsule, rotation=<rot 0 -90 0>}]
 base_tie_l [{shapeType=Capsule, length=0.505, radius=0.505, shapeName=base_tie_l_capsule, rotation=<rot 0 -90 0>}]
 base_tie_r [{shapeType=Capsule, length=0.505, radius=0.505, shapeName=base_tie_r_capsule, rotation=<rot 0 -90 0>}]
 dyn_lace_00_l [{shapeType=Capsule, length=0.5, radius=0.9, shapeName=dyn_lace_00_l_capsule, center=<0.8 0 0>, rotation=<rot 90 0 0>}]
 dyn_lace_00_r [{shapeType=Capsule, length=0.5, radius=0.9, shapeName=dyn_lace_00_r_capsule, center=<0.8 0 0>, rotation=<rot 90 0 0>}]
 dyn_lace_01_l [{shapeType=Capsule, length=0.5, radius=0.9, shapeName=dyn_lace_01_l_capsule, center=<0.8 0 0>, rotation=<rot 90 0 0>}]
 dyn_lace_01_r [{shapeType=Capsule, length=0.5, radius=0.9, shapeName=dyn_lace_01_r_capsule, center=<0.8 0 0>, rotation=<rot 90 0 0>}]
 dyn_lace_l [{shapeType=Capsule, length=0.5, radius=0.9, shapeName=dyn_lace_l_capsule, center=<0.8 0 0>, rotation=<rot 90 0 0>}]
 dyn_lace_r [{shapeType=Capsule, length=0.5, radius=0.9, shapeName=dyn_lace_r_capsule, center=<0.8 0 0>, rotation=<rot 90 0 0>}]
 dyn_tie_l [{shapeType=Capsule, length=2, radius=2, shapeName=dyn_tie_l_capsule, center=<2 0 0>, rotation=<rot 90 0 0>}]
 dyn_tie_r [{shapeType=Capsule, length=2, radius=2, shapeName=dyn_tie_r_capsule, center=<2 0 0>, rotation=<rot 90 0 0>}]
 handle [{shapeType=Capsule, length=20, radius=1.2, shapeName=handle_capsule, center=<-2.548 0.1173 -0.0273>, rotation=<rot 0 -90.2393 -89.8106>}]
}
constraints [
 {bone1Name=dyn_lace_l, bone2Name=base_lace_l, swing1LimitDegrees=25, swing2LimitDegrees=25, twistLimitDegrees=1, twistMotion=Locked}
 {bone1Name=dyn_lace_00_l, bone2Name=dyn_lace_l, swing1LimitDegrees=35, swing2LimitDegrees=35, twistLimitDegrees=1, twistMotion=Locked}
 {bone1Name=dyn_lace_01_l, bone2Name=dyn_lace_00_l, swing1LimitDegrees=35, swing2LimitDegrees=35, twistLimitDegrees=1, twistMotion=Locked}
 {bone1Name=dyn_lace_r, bone2Name=base_lace_r, swing1LimitDegrees=50, swing2LimitDegrees=35, twistLimitDegrees=1, twistMotion=Locked}
 {bone1Name=dyn_lace_00_r, bone2Name=dyn_lace_r, swing1LimitDegrees=40, swing2LimitDegrees=40, twistLimitDegrees=1, twistMotion=Locked}
 {bone1Name=dyn_lace_01_r, bone2Name=dyn_lace_00_r, swing1LimitDegrees=35, swing2LimitDegrees=35, twistLimitDegrees=1, twistMotion=Locked}
 {bone1Name=dyn_tie_l, bone2Name=base_tie_l, swing1LimitDegrees=2, swing2LimitDegrees=7, twistLimitDegrees=1, twistMotion=Locked}
 {bone1Name=dyn_tie_r, bone2Name=base_tie_r, swing1LimitDegrees=2, swing2LimitDegrees=7, twistLimitDegrees=1, twistMotion=Locked}
 {
  bone1Name=base_tie_l
  bone2Name=handle
  swing1LimitDegrees=45
  swing1Motion=Free
  swing2LimitDegrees=45
  swing2Motion=Free
  twistLimitDegrees=45
  twistMotion=Free
 }
 {
  bone1Name=base_tie_r
  bone2Name=handle
  swing1LimitDegrees=45
  swing1Motion=Free
  swing2LimitDegrees=45
  swing2Motion=Free
  twistLimitDegrees=45
  twistMotion=Free
 }
 {
  bone1Name=base_lace_l
  bone2Name=handle
  swing1LimitDegrees=45
  swing1Motion=Free
  swing2LimitDegrees=45
  swing2Motion=Free
  twistLimitDegrees=45
  twistMotion=Free
 }
 {
  bone1Name=base_lace_r
  bone2Name=handle
  swing1LimitDegrees=45
  swing1Motion=Free
  swing2LimitDegrees=45
  swing2Motion=Free
  twistLimitDegrees=45
  twistMotion=Free
 }
]
